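# Material function for Hermes rocks that layers glowing Slurp drips and a ground gradient onto incoming material attributes, behind a static switch.
# Custom primitive data selects a 2x2 atlas cell, toggles painted drips and adjusts the gradient.
# Drip masks drive color, roughness and sun-modulated emissive, while a pivot-relative height gradient tints color and can bend normals upward.
# Platform switches vary gradient cost, and MaterialProxyReplace supplies a simplified base color for proxies.
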MaterialFunction /BRCoreMaterials/Hermes/Rocks/Materials/MF_SlurpDripping
graph {nodes [
  n1 = NamedRerouteUsage()
  n2 = NamedRerouteUsage()
  Add(A=@n1, B=@n2)
  n3 = Abs(None=NamedRerouteUsage())
  n4 = Add(A=ComponentMask(None=@n3) {r=true}, B=ComponentMask(None=@n3) {g=true})
  n5 = Desaturation(None=NamedRerouteUsage())
  n6 = Multiply(A=@n5, B=@n5)
  n7 = Multiply(A=Multiply(A=@n6, B=@n6)) {constB=80}
  Add(A=Multiply(A=@n4, B=@n4), B=@n7)
  Constant()
  n8 = Add(
   A=VertexNormalWS()
   B=Multiply(
    A=Constant3Vector() {constant=<rgba 0 0 1 0>}
    B=ScalarParameter() {defaultValue=0.0803, group=Slurp Ground Gradient, parameterName=Normal Bend Up 2, sortPriority=64}
   )
  ) {constB=20}
  n9 = LinearInterpolate(
   B=Multiply(A=NamedRerouteUsage(), B=ScalarParameter() {group=Slurp Ground Gradient, parameterName=Normal Bend Intensity})
   Alpha=Saturate(None=ScalarParameter() {group=Slurp Ground Gradient, parameterName="NotPerf Gradient? (Switch)", sliderMax=1})
  )
  n10 = NamedRerouteUsage()
  n11 = Saturate(None=Multiply(A=NamedRerouteUsage()) {constB=0.02})
  n12 = ScalarParameter() {group=Slurp Ground Gradient, parameterName=Debug Gradient Mask, sliderMax=1}
  n13 = GetMaterialAttributes(None=NamedRerouteDeclaration(None=FunctionInput() {inputName=MA, inputType=FunctionInput_MaterialAttributes}) {name=MA, nodeColor=<rgba 0 1 0.6484 1>}) {attributeGetTypes [
    #1
    #2
    #3
    #4
    #5
   ]}
  n14 = SetMaterialAttributes!(
   MaterialAttributes=@n13.MaterialAttributes
   BaseColor=NamedRerouteUsage()
   Roughness=NamedRerouteUsage()
   Normal=LinearInterpolate(A=NamedRerouteUsage(), B=Normalize(VectorInput=@n8), Alpha=Saturate(None=@n9))
   EmissiveColor=LinearInterpolate(A=@n10, B=Add(A=@n10, B=@n11), Alpha=Round(None=@n12))
   Metallic
   Specular
   WorldPositionOffset
  ) {attributeSetTypes [
    #1
    #4
    #5
    #6
    #2
    #3
    #7
   ]}
  n15 = LinearInterpolate(
   B=ScalarParameter() {defaultValue=0.5, group=Slurp, parameterName=Leak Color Proxy, sliderMax=1}
   Alpha=Saturate(None=Multiply(A=@n1) {constB=0.05})
  ) {constA=1}
  n16 = Multiply(A=NamedRerouteUsage()) {constB=0.9}
  n17 = Multiply(A=VectorParameter() {group=Slurp, parameterName=Emissive Slurp, defaultValue=<rgba 0 10 10 1>}.RGB) {constB=0.15}
  n18 = Multiply(A=@n2) {constB=0.3}
  n19 = SetMaterialAttributes!(
   MaterialAttributes=@n14
   BaseColor=Add(A=Multiply(A=@n15, B=@n16), B=Multiply(A=@n17, B=@n18))
   Roughness
   Normal
   EmissiveColor
   Metallic
   Specular
  ) {attributeSetTypes [
    #1
    #4
    #5
    #6
    #2
    #3
   ]}
  FunctionOutput(None=StaticSwitchParameter(True=MaterialProxyReplace(Realtime=@n14, MaterialProxy=@n19), False=NamedRerouteUsage()) {group=Slurp, parameterName=UseSlurpDrip?})
  n20 = Add(A=NamedRerouteUsage(), B=@n18)
  Multiply(A=@n20, B=@n20) {constB=5}
  Multiply(A=@n1, B=@n1) {constB=0}
  n21 = Constant() {r=2}
  n22 = AppendVector(A=@n21, B=@n21)
  n23 = Divide(A=TextureCoordinate(), B=@n22)
  n24 = AppendVector(A=Multiply(A=@n21, B=@n21), B=@n21)
  n25 = Frac(None=Clamp(None=ScalarParameter() {bUseCustomPrimitiveData=true, parameterName=DripTextureAtlas, primitiveDataIndex=2, sliderMax=1}) {maxDefault=1}) {desc=0-1 animation}
  n26 = AppendVector(A=@n25, B=@n25)
  n27 = Divide(A=Floor(None=Multiply(A=@n24, B=@n26)), B=@n22)
  n28 = LinearInterpolate(
   A=Frac(None=Add(A=@n23, B=@n27))
   B=TextureCoordinate()
   Alpha=Sign(None=ScalarParameter() {group=Slurp, parameterName="Use Regular UVs? (Switch)", sliderMax=1, sortPriority=1})
  )
  n29 = ScalarParameter() {defaultValue=75, group="A_L1 (TD slot 1)", parameterName=L1 WS_Triplanar Fade_Low/Mobile, sortPriority=0}
  n30 = ScalarParameter() {defaultValue=10, group="A_L1 (TD slot 1)", parameterName=L1 WS_Triplanar Fade, sortPriority=0}
  n31 = MaterialFunctionCall!(
   "Normal (WS)"
   Texture Scale=NamedRerouteDeclaration(None=ScalarParameter() {defaultValue=2000, group="A_L1 (TD slot 1)", parameterName=L1 WS_Tile, sortPriority=0}) {name=L1 WS_Tile, nodeColor=<rgba 1 0 0.75 1>}
   Triplanar Fade=MaterialFunctionCall!(Low=@n29, High=@n30, Epic=@n30, --------, Mobile=@n29, Medium=@n30) {materialFunction=/FortniteBaseMaterials/Functions/MF_QualitySwitch_Scalar}.Result
   Derivatives Scale
  ) {materialFunction=/Game/Athena/Environments/MaterialFunctions/MF_TriplanarDithered}
  n32 = TextureSampleParameter2D!(UVs=@n28, "DDX(UVs)"=@n31.DDX, "DDY(UVs)"=@n31.DDX, Apply View MipBias) {
   group=Slurp
   mipValueMode=TMVM_Derivative
   parameterName=Slurp drip texture B
   samplerType=SAMPLERTYPE_LinearColor
   texture=&/T_Drip_Mask_mud
  }
  n33 = ScalarParameter() {defaultValue=1, group=Slurp, parameterName=Slurp Intensity}
  n34 = ScalarParameter() {defaultValue=1, group=Slurp, parameterName=Mask Curvature}
  n35 = TextureSampleParameter2D!(UVs=@n28, Apply View MipBias) {
   group=Slurp
   parameterName=Slurp drip texture A
   samplerType=SAMPLERTYPE_LinearColor
   texture=&T_Drip_Mask_mud
  }
  n36 = Multiply(A=@n35.G, B=@n35.G)
  n37 = Multiply(A=@n36, B=@n36)
  n38 = Add(A=@n35.G, B=@n37)
  n39 = ScalarParameter() {defaultValue=1.2, group=Slurp, parameterName=Paint Mask}
  n40 = NamedRerouteUsage()
  n41 = Multiply(A=Multiply(A=Multiply(A=@n32.B, B=@n34), B=Multiply(A=@n38, B=@n39) {constB=1.2}), B=Add(A=Add(A=@n7, B=@n40)) {constB=0.1})
  n42 = Saturate(None=Round(None=ScalarParameter() {bUseCustomPrimitiveData=true, group=Slurp, parameterName=Painted Drip, primitiveDataIndex=1, sliderMax=1}))
  NamedRerouteDeclaration(None=Add(A=Multiply(A=@n32.R, B=@n33), B=LinearInterpolate(B=@n41, Alpha=@n42))) {name=Leak Mask, nodeColor=<rgba 1 0 0.4453 1>}
  n43 = NamedRerouteUsage()
  n44 = Saturate(None=Multiply(A=NamedRerouteUsage(), B=ScalarParameter() {defaultValue=3, group=Slurp, parameterName=Spec}) {constB=2})
  NamedRerouteDeclaration(None=LinearInterpolate(A=Multiply(A=@n43) {constB=0.1}, B=Add(A=@n43, B=@n44) {constB=0.5}, Alpha=NamedRerouteUsage())) {name=Spec, nodeColor=<rgba 0.0312 0 1 1>}
  NamedRerouteDeclaration(None=@n13.Roughness) {name=Rough OG, nodeColor=<rgba 0 0.8594 1 1>}
  NamedRerouteDeclaration(None=@n13.Specular) {name=Spec OG, nodeColor=<rgba 0 0.8594 1 1>}
  NamedRerouteDeclaration(None=Multiply(A=@n37, B=@n42)) {name=Painted Leak, nodeColor=<rgba 0 1 0.3672 1>}
  NamedRerouteDeclaration(None=FunctionInput(Preview=NamedRerouteDeclaration(None=@n13.BaseColor) {name=Base Color OG, nodeColor=<rgba 0.3359 0 1 1>}) {bUsePreviewValueAsDefault=true}) {name=Diffuse Color, nodeColor=<rgba 0.9453 1 0 1>}
  n45 = Add(
   A=WorldPosition().Z
   B=Add(
    A=ScalarParameter() {defaultValue=-5, group=Slurp Ground Gradient, parameterName=Gradient Start Height, sortPriority=64}
    B=ScalarParameter() {bUseCustomPrimitiveData=true, group=Slurp, parameterName=HeightMasks Custom, primitiveDataIndex=3, sortPriority=64}
   )
  )
  n46 = ComponentMask(None=MaterialFunctionCall() {materialFunction=&ObjectPivotPoint}.Object Pivot Location) {b=true}
  n47 = OneMinus(None=Divide(
    A=Subtract(A=@n45, B=@n46)
    B=ScalarParameter() {defaultValue=300, group=Slurp Ground Gradient, parameterName=Gradient Size, sortPriority=64}
   ) {constB=500})
  n48 = Saturate(None=@n47)
  n49 = Constant3Vector() {constant=<rgba 0 0 1 0>}
  n50 = ScalarParameter() {defaultValue=0.0803, group=Slurp Ground Gradient, parameterName=Normal Up Mask, sortPriority=64}
  n51 = Normalize(VectorInput=Add(A=VertexNormalWS(), B=Multiply(A=@n49, B=@n50)) {constB=20})
  n52 = Saturate(None=DotProduct(A=@n51, B=Constant3Vector() {constant=<rgba 0 0 1 0>}))
  n53 = Multiply(A=Subtract(A=@n47, B=ScalarParameter() {defaultValue=0.6095, group=Slurp Ground Gradient, parameterName=Upfacing Mask Size, sortPriority=64})) {constB=2}
  n54 = Multiply(
   A=LinearInterpolate(A=Multiply(A=@n48, B=@n52), B=@n48, Alpha=Saturate(None=@n53)) {constA=1}
   B=ScalarParameter() {defaultValue=2, group=Slurp Ground Gradient, parameterName=Gradient Mult, sortPriority=64}
  )
  NamedRerouteDeclaration(None=VertexInterpolator(VS=Saturate(None=@n54)).PS) {name=Mask gradient, nodeColor=<rgba 0 1 0.4375 1>}
  NamedRerouteDeclaration(None=@n51) {name=Bend Up normal, nodeColor=<rgba 1 0.5156 0 1>}
  n55 = NamedRerouteUsage()
  n56 = Add(A=NamedRerouteUsage()) {constB=-0.4}
  n57 = NamedRerouteUsage()
  NamedRerouteDeclaration(None=LinearInterpolate(A=LinearInterpolate(A=@n55, B=@n56, Alpha=@n44) {constB=-0.5}, B=Multiply(A=@n57) {constB=2}, Alpha=NamedRerouteUsage())) {name=Rough, nodeColor=<rgba 0.2891 1 0 1>}
  n58 = NamedRerouteUsage()
  n59 = Multiply(A=Saturate(None=Multiply(A=@n58, B=@n58)), B=NamedRerouteUsage())
  n60 = AtmosphericLightVector()
  n61 = VertexNormalWS()
  n62 = LinearInterpolate(B=ScalarParameter() {defaultValue=0.5, group=Slurp, parameterName=Sun}, Alpha=Saturate(None=DotProduct(A=@n60, B=@n61))) {constA=1, constB=1.5}
  NamedRerouteDeclaration(None=Multiply(A=Multiply(A=@n59, B=@n62), B=VectorParameter() {group=Slurp, parameterName=Emissive Slurp, defaultValue=<rgba 0 10 10 1>}.RGB)) {name=Emissive, nodeColor=<rgba 1 0 0.6094 1>}
  n63 = Multiply(A=NamedRerouteUsage()) {constB=3}
  n64 = Multiply(
   A=NamedRerouteUsage()
   B=LinearInterpolate(B=ScalarParameter() {defaultValue=0.3696, group=Slurp, parameterName=Leak Color, sliderMax=1}, Alpha=Saturate(None=@n63)) {constA=1}
  )
  n65 = VectorParameter() {group=Slurp Ground Gradient, parameterName=Ground Color, sortPriority=64, defaultValue=<rgba 1 1 1 1>}
  n66 = Desaturation(
   None=@n65.RGB
   Fraction=ScalarParameter() {bUseCustomPrimitiveData=true, group=Slurp Ground Gradient, parameterName=Gradrient DeSaturation, primitiveDataIndex=4}
  )
  n67 = Add(A=ScalarParameter() {bUseCustomPrimitiveData=true, group=Slurp Ground Gradient, parameterName=Gradrient Color Multiply, primitiveDataIndex=5})
  n68 = LinearInterpolate(A=@n64, B=Multiply(A=@n64, B=Multiply(A=@n66, B=@n67)), Alpha=NamedRerouteUsage())
  n69 = LinearInterpolate(
   A=@n64
   B=@n68
   Alpha=Saturate(None=ScalarParameter() {group=Slurp Ground Gradient, parameterName="NotPerf Gradient? (Switch)", sliderMax=1})
  )
  NamedRerouteDeclaration(None=MaterialFunctionCall!("Xb1,PS4"=@n69, "PC-High&&Epic,XSX,PS5"=@n69, "PC-Perf,Android"=@n68, Switch=@n69, "PC-Low(NotPerf)&&Medium"=@n69) {materialFunction=/Game/Effects/Niagara/MasterMaterials/Functions/Common/PlatformSwitch_3D}.Result) {name=base color, nodeColor=<rgba 0 0.9297 1 1>}
  NamedRerouteDeclaration(None=@n13.Normal) {name=Normal OG, nodeColor=<rgba 0.7344 0 1 1>}
  NamedRerouteDeclaration(None=Multiply(A=@n59) {constB=8}) {name=Emissive Mask, nodeColor=<rgba 0 0.8828 1 1>}
  n70 = Constant()
  n71 = FunctionInput(Preview=Constant()) {bUsePreviewValueAsDefault=true, inputName=L3 Gradient, inputType=FunctionInput_Scalar}
  n72 = Saturate(None=Subtract(A=NamedRerouteDeclaration(None=@n71) {name=L3 gradient, nodeColor=<rgba 0.6875 0 1 1>}, B=NamedRerouteUsage()))
  NamedRerouteDeclaration(None=MaterialFunctionCall("Xb1,PS4"=@n70, "PC-High&&Epic,XSX,PS5"=@n70, "PC-Perf,Android"=@n72, Switch=@n72, "PC-Low(NotPerf)&&Medium"=@n70) {materialFunction=&PlatformSwitch_1D}.Result) {name=Gradient Dry, nodeColor=<rgba 0.4297 0 1 1>}
  NamedRerouteUsage()
  NamedRerouteUsage()
  NamedRerouteUsage()
  NamedRerouteUsage()
  ScalarParameter() {defaultValue=1, group=Slurp, parameterName=Emissive Mask}
 ]}
