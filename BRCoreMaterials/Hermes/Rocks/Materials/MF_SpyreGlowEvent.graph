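# Material function that adds a pulsing inner-core emissive glow to Hermes Spyre rocks for the Mask Maker event, behind a static switch.
# The fake core combines a camera-facing virtual-plane ball with a view-dot term, shaped by up/down parameter sets and camera-height masks.
# Scrolling noise and bump-offset parallax build interior detail, while world-position-offset sines vary the pulse.
# An MPC slider gated by custom primitive data provides event progress; platform and quality switches adjust cost and intensity.
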MaterialFunction /BRCoreMaterials/Hermes/Rocks/Materials/MF_SpyreGlowEvent
graph {
 properties {description=tr}
 nodes [
  n1 = FunctionInput() {inputName=MA, inputType=FunctionInput_MaterialAttributes}
  n2 = GetMaterialAttributes(None=@n1) {attributeGetTypes [#1]}
  n3 = Add(A=@n2.EmissiveColor, B=NamedRerouteUsage())
  FunctionOutput(None=StaticSwitchParameter(
    True=SetMaterialAttributes!(MaterialAttributes=@n2.MaterialAttributes, EmissiveColor=@n3) {attributeSetTypes [#1]}
    False=@n1
   ) {group=Mask Maker Event, parameterName=Mask Maker Event, sortPriority=1}) {outputName=MA Spyre Emissive Glow}
  n4 = LocalPosition()
  n5 = ComponentMask(None=@n4.XYZ) {b=true, g=true}
  n6 = Abs(None=VertexNormalWS())
  n7 = MaterialFunctionCall(Float3=@n6) {materialFunction=&BreakOutFloat3Components}
  LinearInterpolate(
   A=LinearInterpolate(A=LinearInterpolate(B=@n5, Alpha=@n7.R), B=ComponentMask(None=@n4.XYZ) {b=true, r=true}, Alpha=@n7.G)
   B=ComponentMask(None=@n4.XYZ) {g=true, r=true}
   Alpha=@n7.B
  )
  n8 = Constant() {r=0.0003}
  Multiply(A=Time(), B=@n8) {constB=0.1}
  NamedRerouteDeclaration(None=FunctionInput() {inputName=Roughness, inputType=FunctionInput_Scalar}) {name=Rough, nodeColor=<rgba 0.2188 0 1 1>}
  n9 = Multiply(A=NamedRerouteUsage()) {constB=0.2}
  n10 = Constant() {r=0.075}
  n11 = Add(A=Multiply(A=Time(), B=@n10) {constB=0.1}, B=@n10)
  n12 = MaterialFunctionCall(Float2=MaterialFunctionCall!(UV Coord=NamedRerouteUsage(), Height=Constant() {r=5}, ReferencePlane=NamedRerouteUsage(), HeightRatio=NamedRerouteUsage(), Normal) {materialFunction=/Game/Global/MaterialFunctions/MF_BumpOffset_Accurate}.Result) {materialFunction=&BreakOutFloat2Components}
  n13 = Frac(None=@n8)
  n14 = Add(A=@n12.G, B=@n13)
  n15 = Multiply(
   A=ConstantBiasScale(None=Sine(None=@n11))
   B=TextureSample(
    UVs=MaterialFunctionCall(X=@n12.R, Y=@n14) {materialFunction=&MakeFloat2}.Result
    Tex=NamedRerouteUsage()
   ) {samplerSource=SSM_Wrap_WorldGroupSettings, samplerType=SAMPLERTYPE_Masks}.G
  )
  n16 = NamedRerouteUsage()
  n17 = MaterialFunctionCall(Float2=MaterialFunctionCall(UV Coord=Multiply(A=@n16) {constB=0.2}, Height=Constant() {r=2}, ReferencePlane=NamedRerouteUsage(), HeightRatio=NamedRerouteUsage()) {materialFunction=&MF_BumpOffset_Accurate}.Result) {materialFunction=&BreakOutFloat2Components}
  n18 = MaterialFunctionCall(X=@n17.R, Y=Add(A=@n13, B=@n17.G)) {materialFunction=&MakeFloat2}
  n19 = NamedRerouteUsage()
  n20 = Sine(None=Add(
    A=Add(A=@n10) {constB=0.3}
    B=TextureSample(UVs=@n18.Result, Tex=@n19) {samplerSource=SSM_Wrap_WorldGroupSettings, samplerType=SAMPLERTYPE_Masks}.R
   ))
  n21 = Add(A=@n15, B=Multiply(A=ConstantBiasScale(None=@n20)))
  n22 = Add(
   A=QualitySwitch(Default=@n21, Low=@n15, High=@n21, Medium=@n15, Epic=@n21)
   B=LinearInterpolate(A=NamedRerouteUsage(), B=NamedRerouteUsage(), Alpha=NamedRerouteUsage())
  ) {constB=0.01}
  NamedRerouteDeclaration(None=Saturate(None=Add(A=@n9, B=@n22))) {name=Rock Interior, nodeColor=<rgba 0.6172 1 0 1>}
  NamedRerouteDeclaration(None=LinearInterpolate(A=ScalarParameter() {defaultValue=0.3, group=Mask Maker Event, parameterName=Int Sine, sortPriority=2}, B=Constant() {r=0.45}, Alpha=Round())) {name=Int Sine, nodeColor=<rgba 0 1 0.0625 1>}
  n23 = NamedRerouteUsage()
  NamedRerouteDeclaration(None=LinearInterpolate(
    A=ScalarParameter() {defaultValue=1, group=Mask Maker Event, parameterName=Down Shape Size X, sortPriority=3}
    B=Constant() {r=0.15}
    Alpha=@n23
   )) {name=Down Shape Size X, nodeColor=<rgba 0 1 0.0625 1>}
  NamedRerouteDeclaration(None=LinearInterpolate(
    A=ScalarParameter() {defaultValue=3.2, group=Mask Maker Event, parameterName=Up Shape size X, sortPriority=3}
    B=Constant() {r=1.5}
    Alpha=@n23
   )) {name=Up Shape size X, nodeColor=<rgba 0 1 0.0625 1>}
  NamedRerouteDeclaration(None=Constant() {r=0.1}) {name=HeightRatio, nodeColor=<rgba 1 0 0.4688 1>}
  n24 = NamedRerouteUsage()
  NamedRerouteDeclaration(None=LinearInterpolate(
    A=ScalarParameter() {defaultValue=2, group=Mask Maker Event, parameterName=Down Shape Size Y, sortPriority=3}
    B=Constant() {r=2.7}
    Alpha=@n24
   )) {name=Down Shape Size Y, nodeColor=<rgba 0 1 0.0625 1>}
  NamedRerouteDeclaration(None=LinearInterpolate(
    A=ScalarParameter() {defaultValue=1.6, group=Mask Maker Event, parameterName=Up Shape size Y, sortPriority=3}
    B=Constant() {r=2.1}
    Alpha=@n24
   )) {name=Up Shape size Y, nodeColor=<rgba 0 1 0.0625 1>}
  NamedRerouteDeclaration(None=LinearInterpolate(
    A=ScalarParameter() {defaultValue=40, group=Mask Maker Event, parameterName=Ball Intensity Down, sortPriority=4}
    B=Constant() {r=100}
    Alpha=@n24
   )) {name=Ball Intensity Down, nodeColor=<rgba 0 1 0.0625 1>}
  NamedRerouteDeclaration(None=LinearInterpolate(
    A=ScalarParameter() {defaultValue=25, group=Mask Maker Event, parameterName=Ball Intensity Up, sortPriority=4}
    B=Constant() {r=30}
    Alpha=@n24
   )) {name=Ball Intensity Up, nodeColor=<rgba 0 1 0.0625 1>}
  NamedRerouteDeclaration(None=LinearInterpolate(
    A=ScalarParameter() {defaultValue=0.1, group=Mask Maker Event, parameterName=Min interior bright, sortPriority=4}
    B=Constant() {r=0.2}
    Alpha=@n24
   )) {name=Min interior bright, nodeColor=<rgba 0 1 0.0625 1>}
  NamedRerouteDeclaration(None=LinearInterpolate(
    A=ScalarParameter() {defaultValue=0.2, group=Mask Maker Event, parameterName=Min interior bright up, sortPriority=4}
    B=Constant() {r=0.2}
    Alpha=@n24
   )) {name=Min interior bright up, nodeColor=<rgba 0 1 0.0625 1>}
  n25 = PixelDepth()
  n26 = Constant() {r=9910.1113}
  NamedRerouteDeclaration(None=Saturate(None=Divide(A=@n25.R, B=@n26))) {name=Distance, nodeColor=<rgba 0.2656 1 0 1>}
  n27 = NamedRerouteUsage()
  n28 = NamedRerouteUsage()
  n29 = NamedRerouteUsage()
  n30 = ComponentMask(None=VertexNormalWS()) {g=true, r=true}
  n31 = Constant2Vector()
  n32 = LinearInterpolate(A=Constant() {desc=Core Height, r=750}, B=Constant() {desc=Core Height, r=300}, Alpha=NamedRerouteUsage())
  n33 = MaterialFunctionCall(
   Plane Normal Axis=CameraVectorWS()
   Plane X Axis=Constant3Vector() {constant=<rgba 0 0 1 0>}
   Plane Center=Subtract(A=ObjectPositionWS(), B=AppendVector(A=@n31, B=@n32))
   UV Scale=Constant() {r=1}
  ) {materialFunction=&VirtualPlaneCoordinates}
  n34 = MaterialFunctionCall(Float2=Add(A=Multiply(A=@n30) {constB=0.03}, B=@n33.UVs)) {materialFunction=&BreakOutFloat2Components}
  n35 = Constant() {r=-1.1}
  n36 = NamedRerouteUsage()
  n37 = NamedRerouteUsage()
  n38 = NamedRerouteUsage()
  n39 = AppendVector(
   A=Multiply(A=LinearInterpolate(A=@n27, B=@n28, Alpha=@n29), B=Add(A=@n34.R, B=@n35) {constB=-0.9})
   B=Multiply(A=Add(A=@n34.G) {constB=-2}, B=LinearInterpolate(A=@n36, B=@n37, Alpha=@n38))
  )
  n40 = MaterialFunctionCall(Float2=Multiply(A=@n39, B=@n39)) {materialFunction=&BreakOutFloat2Components}
  n41 = Divide(
   A=SquareRoot(None=Add(A=@n40.R, B=@n40.R))
   B=LinearInterpolate(
    A=ScalarParameter() {defaultValue=500, group=Mask Maker Event, parameterName=Ball Size Down, sortPriority=50}
    B=ScalarParameter() {defaultValue=500, group=Mask Maker Event, parameterName=Ball Size Up, sortPriority=50}
    Alpha=NamedRerouteUsage()
   )
  )
  n42 = LinearInterpolate(
   A=ScalarParameter() {defaultValue=0.02, group=Mask Maker Event, parameterName=Ball Contrast Down}
   B=ScalarParameter() {defaultValue=0.2, group=Mask Maker Event, parameterName=Ball Contrast Up}
   Alpha=NamedRerouteUsage()
  )
  n43 = OneMinus(None=Saturate(None=Power(Base=@n41, Exp=@n42)))
  NamedRerouteDeclaration(None=@n43) {name=Ball, nodeColor=<rgba 1 0.7734 0 1>}
  n44 = Divide(A=NamedRerouteUsage(), B=ScalarParameter() {defaultValue=556.6693, group=Mask Maker Event, parameterName=Upward Motion Size}) {constB=125}
  n45 = Add(
   A=ComponentMask(None=@n44) {r=true}
   B=Divide(A=Time(), B=ScalarParameter() {defaultValue=-9, group=Mask Maker Event, parameterName=Upward Motion Time}) {constB=-9}
  )
  n46 = Add(A=ComponentMask(None=@n44) {g=true}) {constB=0.37}
  n47 = ComponentMask(None=PixelNormalWS()) {g=true, r=true}
  n48 = TextureSample(UVs=Add(A=AppendVector(A=@n45, B=@n46), B=Multiply(A=@n47) {constB=1.2}), Tex=@n19) {samplerSource=SSM_Wrap_WorldGroupSettings, samplerType=SAMPLERTYPE_Masks}
  n49 = NamedRerouteUsage()
  NamedRerouteDeclaration(None=Add(A=Multiply(A=@n48.R, B=@n48.R), B=Multiply(A=@n49) {constB=0.4})) {name=Up Noise, nodeColor=<rgba 0.0078 0 1 1>}
  NamedRerouteDeclaration(None=@n33.UVs) {nodeColor=<rgba 0.9453 1 0 1>}
  n50 = NamedRerouteUsage()
  n51 = CameraVectorWS()
  n52 = MaterialFunctionCall(Float3=@n51) {materialFunction=&BreakOutFloat3Components}
  n53 = ObjectPositionWS()
  n54 = ComponentMask(None=Normalize(VectorInput=@n53)) {b=true}
  n55 = MaterialFunctionCall(Float3=@n51) {materialFunction=&BreakOutFloat3Components}
  n56 = Constant() {r=0.312}
  n57 = LinearInterpolate(
   A=LinearInterpolate(
    A=@n51
    B=MaterialFunctionCall(X=@n52.R, Y=@n52.R, Z=@n54) {materialFunction=&MakeFloat3}.Result
    Alpha=NamedRerouteUsage()
   )
   B=MaterialFunctionCall(X=@n55.R, Y=@n55.R, Z=Add(A=@n54, B=@n56)) {materialFunction=&MakeFloat3}.Result
   Alpha=NamedRerouteUsage()
  )
  n58 = Subtract(A=WorldPosition().XYZ, B=@n53)
  n59 = AppendVector(A=Constant2Vector(), B=Constant() {desc=Core Height, r=-750})
  n60 = LinearInterpolate(
   A=Normalize(VectorInput=Subtract(A=@n58, B=@n59))
   B=PixelNormalWS()
   Alpha=LinearInterpolate(A=Constant() {r=0.3}, B=Constant() {r=0.6}, Alpha=NamedRerouteUsage())
  ) {constAlpha=0.3}
  n61 = NamedRerouteUsage()
  n62 = NamedRerouteUsage()
  n63 = NamedRerouteUsage()
  n64 = Saturate(None=Power(Base=DotProduct(A=@n57, B=@n60), Exp=LinearInterpolate(A=@n61, B=@n62, Alpha=@n63)))
  n65 = Multiply(A=Multiply(A=@n50, B=@n64)) {constB=0.05}
  n66 = Multiply(A=OneMinus(None=@n50), B=NamedRerouteUsage())
  n67 = NamedRerouteUsage()
  n68 = Saturate(None=Multiply(A=@n64, B=@n64))
  n69 = Multiply(
   A=Multiply(A=Add(A=@n68) {constB=0.05}, B=@n43)
   B=LinearInterpolate(A=NamedRerouteUsage(), B=NamedRerouteUsage(), Alpha=NamedRerouteUsage())
  )
  n70 = Multiply(
   A=Multiply(A=Add(A=@n65, B=@n66), B=LinearInterpolate(A=@n67, Alpha=@n50))
   B=Multiply(A=VectorParameter() {group=Mask Maker Event, parameterName=Core Color, defaultValue=<rgba 0.3594 0 0.3023 1>}.RGB, B=Add(A=@n68, B=@n69))
  )
  n71 = NamedRerouteUsage()
  n72 = Multiply(A=Clamp(None=Multiply(A=@n70, B=@n71), Max=NamedRerouteUsage()) {clampMode=CMODE_ClampMax, maxDefault=3}, B=NamedRerouteUsage())
  n73 = MaterialFunctionCall(Float3=WorldPosition().XYZ) {materialFunction=&BreakOutFloat3Components}
  n74 = Add(A=Add(A=@n73.R, B=@n73.R), B=@n73.B)
  n75 = Multiply(A=Add(A=Time(), B=Divide(A=@n74) {constB=10000, desc=leap of faith}), B=Constant() {r=0.4})
  n76 = LinearInterpolate(Alpha=ConstantBiasScale(None=Sine(None=@n75))) {constA=0.4, constB=0.8}
  n77 = NamedRerouteUsage()
  NamedRerouteDeclaration(None=Multiply(A=Multiply(A=@n72, B=@n76), B=Add(A=@n77) {constB=0.1})) {name=Mask Maker Event Monolith, nodeColor=<rgba 0.5234 0 1 1>}
  NamedRerouteDeclaration(None=Constant()) {name=ReferencePlane, nodeColor=<rgba 0 0.7891 1 1>}
  n78 = TextureCoordinate()
  n79 = ComponentMask(None=Constant2Vector() {g=10, r=10}) {g=true, r=true}
  n80 = ComponentMask(None=VertexNormalWS()) {g=true, r=true}
  NamedRerouteDeclaration(None=Add(A=Multiply(A=@n78, B=@n79) {constB=20}, B=Multiply(A=@n80) {constB=0.03})) {name=Rock Interior UVs, nodeColor=<rgba 0.1484 1 0 1>}
  NamedRerouteDeclaration(None=TextureObject() {samplerType=SAMPLERTYPE_Alpha, texture=&/T_RockInterior_Noise}) {name=NoiseTex, nodeColor=<rgba 0.1719 1 0 1>}
  n81 = ObjectPositionWS()
  n82 = Constant3Vector() {constant=<rgba 0 0 -3500 0>}
  n83 = Divide(A=Subtract(A=CameraPositionWS(), B=Add(A=@n81, B=@n82)), B=Constant() {r=1500})
  n84 = OneMinus(None=Saturate(None=ComponentMask(None=@n83) {b=true}))
  n85 = OneMinus(None=NamedRerouteUsage())
  NamedRerouteDeclaration(None=Saturate(None=Multiply(A=@n84, B=@n85))) {name=Bellow mask, nodeColor=<rgba 0 1 0.25 1>}
  NamedRerouteDeclaration(None=FunctionInput() {inputName=Base Color}) {name=L1 Base Color Adjusted, nodeColor=<rgba 0 1 0.6953 1>}
  n86 = CollectionParameter() {collection=&/MPC_Maskmaker_Event, parameterName=Maskmaker Event Slider}
  NamedRerouteDeclaration(None=LinearInterpolate(B=Saturate(None=@n86), Alpha=ScalarParameter() {bUseCustomPrimitiveData=true, parameterName=Event?, primitiveDataIndex=1})) {name=Event Slider, nodeColor=<rgba 0.3828 0 1 1>}
  n87 = Add(A=ActorPositionWS() {originType=CameraRelative}, B=AppendVector(A=Constant2Vector(), B=Constant() {desc=Core Height, r=-750}))
  n88 = ComponentMask(None=Normalize(VectorInput=OneMinus(None=@n87))) {b=true}
  n89 = Power(Base=Multiply(A=@n88, B=Abs(None=@n88)), Exp=ScalarParameter() {defaultValue=6, group=Mask Maker Event, parameterName=Camera Pos Mask}) {constExponent=5}
  n90 = CameraPositionWS()
  n91 = Add(A=ObjectPositionWS(), B=Constant3Vector() {constant=<rgba 0 0 600 0>})
  n92 = ComponentMask(None=Divide(A=Subtract(A=@n90, B=@n91), B=Constant() {r=2000})) {b=true}
  n93 = Multiply(A=NamedRerouteDeclaration(None=Saturate(None=@n92)) {name=Above Mask, nodeColor=<rgba 0 1 0.25 1>}) {constB=0.5}
  NamedRerouteDeclaration(None=Saturate(None=Add(A=@n89, B=@n93))) {name=Camera Pos Mask, nodeColor=<rgba 1 0.5625 0 1>}
  NamedRerouteDeclaration(None=LinearInterpolate(
    A=ScalarParameter() {defaultValue=2, group=Mask Maker Event, parameterName=Fresnel Amount Down, sortPriority=1}
    B=Constant() {r=1.8}
    Alpha=@n23
   )) {name=Fresnel Amount Down, nodeColor=<rgba 0 1 0.0625 1>}
  NamedRerouteDeclaration(None=FunctionInput() {inputName=Gold Mask, inputType=FunctionInput_Scalar}) {name=Crack_GoldMask, nodeColor=<rgba 1 0.5156 0 1>}
  NamedRerouteDeclaration(None=LinearInterpolate(
    A=ScalarParameter() {defaultValue=0.8, group=Mask Maker Event, parameterName=Fresnel Amount Up, sortPriority=1}
    B=Constant() {r=1.5}
    Alpha=@n23
   )) {name=Fresnel Amount Up, nodeColor=<rgba 0 1 0.0625 1>}
  n94 = ScalarParameter() {defaultValue=200, group=Mask Maker Event, parameterName=Core Intensity, sortPriority=2}
  n95 = ScalarParameter() {defaultValue=220, group=Mask Maker Event, parameterName=Core Intensity Mobile, sortPriority=2}
  NamedRerouteDeclaration(None=LinearInterpolate(
    A=MaterialFunctionCall!("Xb1,PS4"=@n94, "PC-High&&Epic,XSX,PS5"=@n94, "PC-Perf,Android"=@n95, Switch=@n95, "PC-Low(NotPerf)&&Medium"=@n95) {materialFunction=/Game/Effects/Niagara/MasterMaterials/Functions/Common/PlatformSwitch_3D}.Result
    B=Constant() {r=700}
    Alpha=@n23
   )) {name=Core Intensity, nodeColor=<rgba 0 1 0.0625 1>}
  NamedRerouteDeclaration(None=LinearInterpolate(
    A=ScalarParameter() {defaultValue=3, group=Mask Maker Event, parameterName=Clamp Max Intensity, sortPriority=2}
    B=Constant() {r=4}
    Alpha=@n23
   )) {name=Clamp Max Intensity, nodeColor=<rgba 0 1 0.0625 1>}
  NamedRerouteUsage()
  NamedRerouteUsage()
  Normalize(VectorInput=@n6)
  ScalarParameter() {defaultValue=6, group=Mask Maker Event, parameterName=Camera Pos Mask}
  ScalarParameter() {defaultValue=9910.1113, parameterName=Distance}
  VertexNormalWS()
 ]
}
